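; Invulnerable building actor that runs a timed, pattern-based fountain show in Hera's desert city when a scalable float enables it.
; Tagged arrow components mark the jets, placing instanced nozzle meshes at construction and building a tag-to-index map at begin play.
; The authority optionally shuffles patterns, starts displays on a looping timer and advances the pattern index after random delays.
; Remotes push pattern indices, heights and delays as arrays into one Niagara system.
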(event EventBeginPlay
  ;[1 1 1] Init
  (Variables|BuildingActor|SetbIsInvulnerable true)
  (if (ScalableFloat|GetValueasBool (Variables|Default|GetFountainsEnabled))
    (CallFunction|GetAllArrows)
    (CallFunction|ShufflePatterns (Variables|Default|GetShouldShufflePatterns))
    (Variables|Default|SetMaxNumPatterns (Utilities|Array|LastIndex (Variables|Default|GetFountainPatterns)))
    (CallFunction|StartEventTimer)))

(event Custom|StartEventTimer
  ;[1 1 1] Play Event
  (Utilities|Time|SetTimerbyEvent (delegate Custom|StartEvent) (Variables|Default|GetDurationBetweenEvents) true))

(event Custom|StartEvent
  ;[1 1 1] Play Event
  (Networking|SwitchHasAuthority
    (:Authority
      (CallFunction|StartDisplayTimer))))

(event Custom|DebugStartFountains
  ;[1 1 1] Play Event
  (Networking|SwitchHasAuthority
    (:Authority
      (CallFunction|StartDisplayTimer))))

(event Custom|StartDisplayTimer
  ;[1 1 1] Display Duration Timer
  (Networking|SwitchHasAuthority
    (:Authority
      (Utilities|FlowControl|Delay 4.0)
      (Variables|Default|SetIsFountainDisplayActive true)
      (Utilities|Time|SetTimerbyEvent (delegate Custom|DisplayTimerFinished) (Variables|Default|GetWaterFountainDisplayDuration))
      (CallFunction|StartFountainsPattern))))

(event Custom|DisplayTimerFinished
  ;[1 1 1] Display Duration Timer
  (Variables|Default|SetIsFountainDisplayActive false))

(event Custom|StartFountainsPattern
  (bind _fountainpatterns (Variables|Default|GetFountainPatterns))
  (bind _output (Utilities|Array|Get(acopy) _fountainpatterns (Variables|Default|GetCurrentPatternPlayingIndex)))
  (bind (_patterntagidentifiers_5_7490a01c4082afa082025686af6460ec _initialdelays_10_f251d6d34614a2226626629e3931cb1c _scales_19_e06da9184f9c84c7b7332bb316c75eaa) (Utilities|Struct|BreakStructFountainDisplay _output))
  (bind _currentpatternplayingindex (Variables|Default|GetCurrentPatternPlayingIndex))
  ;[1 1 1] StartFountains
  (Networking|SwitchHasAuthority
    (:Authority
      (Utilities|FlowControl|Delay (Math|Random|RandomFloatinRange 2.0 3.0)) ; server
      (CallFunction|ResetFountains)
      (CallFunction|PatternFinished))
    (:Remote
      (if (Variables|Default|GetDebugPatternEnabled)
        (Variables|Default|SetCurrentActivePattern (Utilities|Array|Get(acopy) (Variables|Default|GetFountainPatterns) (Variables|Default|GetDebugPatternIndex)))
        (else
          (Variables|Default|SetCurrentActivePattern _output)))
      (Utilities|FlowControl|DoOnce false
        (:Completed
          (bind _allarrowcomponentswith_matchingtag (CallFunction|GetArrowIndexfromTags _patterntagidentifiers_5_7490a01c4082afa082025686af6460ec))
          (Niagara|NiagaraSetInt32Array (Variables|Default|GetNS_DesertCity_Fountain) "PatternArray" _allarrowcomponentswith_matchingtag)
          (Niagara|NiagaraSetFloatArray (Variables|Default|GetNS_DesertCity_Fountain) "HeightArray" _scales_19_e06da9184f9c84c7b7332bb316c75eaa)
          (Niagara|NiagaraSetFloatArray (Variables|Default|GetNS_DesertCity_Fountain) "DelayArray" _initialdelays_10_f251d6d34614a2226626629e3931cb1c)
          (Variables|Default|SetLongestDelay 0.0)
          (for _array_element _initialdelays_10_f251d6d34614a2226626629e3931cb1c
            (if (>= _array_element (Variables|Default|GetLongestDelay))
              (Variables|Default|SetLongestDelay _array_element)))
          (Niagara|SetNiagaraVariable(Float) (Variables|Default|GetNS_DesertCity_Fountain) "LongestDelay" (Variables|Default|GetLongestDelay))
          (Components|Activation|Activate (Variables|Default|GetNS_DesertCity_Fountain) true))))))

(event Custom|ResetFountains
  (bind _fountainpatterns (Variables|Default|GetFountainPatterns))
  (bind _output (Utilities|Array|Get(acopy) _fountainpatterns (Variables|Default|GetCurrentPatternPlayingIndex)))
  (bind (_patterntagidentifiers_5_7490a01c4082afa082025686af6460ec _initialdelays_10_f251d6d34614a2226626629e3931cb1c _scales_19_e06da9184f9c84c7b7332bb316c75eaa) (Utilities|Struct|BreakStructFountainDisplay _output))
  (bind _currentpatternplayingindex (Variables|Default|GetCurrentPatternPlayingIndex))
  ;[1 1 1] StartFountains
  (Networking|SwitchHasAuthority
    (:Remote
      (Utilities|FlowControl|DoOnce false
        (:Completed
          (bind _allarrowcomponentswith_matchingtag (CallFunction|GetArrowIndexfromTags _patterntagidentifiers_5_7490a01c4082afa082025686af6460ec))
          (Niagara|NiagaraSetInt32Array (Variables|Default|GetNS_DesertCity_Fountain) "PatternArray" _allarrowcomponentswith_matchingtag)
          (Niagara|NiagaraSetFloatArray (Variables|Default|GetNS_DesertCity_Fountain) "HeightArray" _scales_19_e06da9184f9c84c7b7332bb316c75eaa)
          (Niagara|NiagaraSetFloatArray (Variables|Default|GetNS_DesertCity_Fountain) "DelayArray" _initialdelays_10_f251d6d34614a2226626629e3931cb1c)
          (Variables|Default|SetLongestDelay 0.0)
          (for _array_element _initialdelays_10_f251d6d34614a2226626629e3931cb1c
            (if (>= _array_element (Variables|Default|GetLongestDelay))
              (Variables|Default|SetLongestDelay _array_element)))
          (Niagara|SetNiagaraVariable(Float) (Variables|Default|GetNS_DesertCity_Fountain) "LongestDelay" (Variables|Default|GetLongestDelay))
          (Components|Activation|Activate (Variables|Default|GetNS_DesertCity_Fountain) true))))))

(event Custom|PatternFinished
  ;[1 1 1] PatternFinished, Next Pattern
  (Networking|SwitchHasAuthority
    (:Authority
      (bind _output_get (Variables|Default|SetCurrentPatternPlayingIndex (+ (Variables|Default|GetCurrentPatternPlayingIndex) 1)))
      (if (> _output_get (Variables|Default|GetMaxNumPatterns))
        (Variables|Default|SetCurrentPatternPlayingIndex 0) ; reset to first index if looping
        (if (not (Variables|Default|GetIsFountainDisplayActive))
          (Development|PrintString "Waterfall Display No Longer Active" true true "(R=0.000000,G=0.660000,B=1.000000,A=1.000000)" 5.0))
        (else
          _))
      (CallFunction|StartFountainsPattern))))

(fn GetAllArrows
  (bind _l_tag (LocalVariables|GetL_Tag))
  (bind _tagtoindexset (Variables|Default|GetTagToIndexSet))
  (bind (_value _returnvalue) (Utilities|Map|Find _tagtoindexset _l_tag))
  (for _array_element (Components|GetChildrenComponents (Variables|Default|GetDefaultSceneRoot) true)
    (bind _asarrow_component (Utilities|Casting|CastToArrowComponent _array_element))
    (Utilities|Array|Add (Variables|Default|GetAllArrowComponents) _asarrow_component)
    (bind (_returnvalue_location _returnvalue_rotation _returnvalue_scale) (Math|Transform|BreakTransform (Transformation|GetWorldTransform _asarrow_component)))
    (Utilities|Array|Add (Variables|Default|GetFountainPositions) _returnvalue_location)
    (Utilities|Array|Add (Variables|Default|GetFountainScales) _returnvalue_scale))
  (Utilities|FlowControl|DoOnce false
    (:Completed
      (Niagara|NiagaraSetPositionArray (Variables|Default|GetNS_DesertCity_Fountain) "PosArray" (Variables|Default|GetFountainPositions))
      (Niagara|NiagaraSetVectorArray (Variables|Default|GetNS_DesertCity_Fountain) "ScaleArray" (Variables|Default|GetFountainScales))
      ; Setup map of tags to arrow component indicies
      (for (_array_element_2 _array_index) (Variables|Default|GetAllArrowComponents)
        (for _array_element_3 (Variables|/Script/Engine.ActorComponent|GetComponentTags _array_element_2)
          (LocalVariables|SetL_Tag _array_element_3)
          (if _returnvalue
            (bind _output_get (LocalVariables|SetL_IntSet (Utilities|Struct|BreakStructIntSet _value)))
            (Utilities|Set|Add _output_get _array_index)
            (Utilities|Map|Add (Variables|Default|GetTagToIndexSet) (LocalVariables|GetL_Tag) (Utilities|Struct|MakeStructIntSet (LocalVariables|GetL_IntSet)))
            (else
              (Utilities|Map|Add (Variables|Default|GetTagToIndexSet) (LocalVariables|GetL_Tag) (Utilities|Struct|MakeStructIntSet (Utilities|Set|MakeSet _array_index))))))))))

(fn GetArrowIndexFromTags
  (params (Tags :type Array%20of%20Names) (Delays :type Array%20of%20Float%20%28double-precision%29s))
  (bind _output_get (LocalVariables|SetTagsLocal Tags))
  (for _array_element _output_get
    (bind _valuepin (Utilities|Set|ForEachLoop(Set) (Utilities|Struct|BreakStructIntSet (Utilities|Map|Find (Variables|Default|GetTagToIndexSet) _array_element))))
    (Utilities|Array|AddUnique (LocalVariables|GetAllMatchingArrows) _valuepin))
  (return (LocalVariables|GetAllMatchingArrows) (LocalVariables|GetSeqDelay)))

(fn GetArrowsFromTags
  (params (Tags :type Array%20of%20Names))
  (LocalVariables|SetTagsLocal Tags)
  (for _array_element (Variables|Default|GetAllArrowComponents)
    (for (_array_element_2 _array_index) (LocalVariables|GetTagsLocal)
      (if (> _array_index 9)
        _
        (elif (Components|ComponentHasTag _array_element _array_element_2)
          (Utilities|Array|Add (LocalVariables|GetAllMatchingArrows) _array_element)))))
  (return (LocalVariables|GetAllMatchingArrows)))

(fn GetInitialDelayAtIndex
  (params (Index :type Integer))
  (bind (_patterntagidentifiers_5_7490a01c4082afa082025686af6460ec _initialdelays_10_f251d6d34614a2226626629e3931cb1c) (Utilities|Struct|BreakStructFountainDisplay (Variables|Default|GetCurrentActivePattern)))
  (return (Utilities|Array|Get(aref) _initialdelays_10_f251d6d34614a2226626629e3931cb1c Index)))

(fn MakeArrowPattern
  (params (Arrows :type Array%20of%20Arrow%20Component%20Object%20References)))

(fn PlayFountainFXAtIndex
  (params (Index :type Integer))
  (bind _all_arrow_components_with_matching_tag (Variables|Default|GetAll%20Arrow%20Components%20with%20Matching%20Tag))
  (bind _output (Utilities|Array|Get(acopy) _all_arrow_components_with_matching_tag Index))
  (bind _returnvalue (Transformation|GetWorldTransform _output))
  (bind _location (.location _returnvalue))
  (Niagara|SpawnSystematLocation :SystemTemplate "/Game/Effects/Fort_Effects/Effects/Environment/Hera/DesertCity/NS_DesertCity_Fountain.NS_DesertCity_Fountain" :Location (+ _location "0.000000,0.000000,40.000000") :Scale (* (.scale _returnvalue) "7.000000,7.000000,3.000000") :bAutoDestroy true :bAutoActivate true :PoolingMethod "None" :bPreCullCheck true)
  (Audio|PlaySoundatLocation "/Game/Sounds/Fort_Test_Sounds/Tones/1k_test_tone_spatialized_Cue.1k_test_tone_spatialized_Cue" _location "0, 0, 0" 1.0 1.0 0.0 "/Game/Athena/Sounds/Props/Abduction_Pipes/Abduction_Pipe_Leak_Splash_Impact_Loop_Att.Abduction_Pipe_Leak_Splash_Impact_Loop_Att"))

(fn ScaleToFountainHeight
  (params (Scales :type Array%20of%20Float%20%28double-precision%29s))
  (for _array_element Scales
    (Utilities|Array|Add (LocalVariables|GetHeightsLocal) (Math|Float|Clamp(Float) (* _array_element (LocalVariables|GetFountainMaxHeight)) (LocalVariables|GetFountainMinHeight) (LocalVariables|GetFountainMaxHeight))))
  (return (LocalVariables|GetHeightsLocal)))

(fn ShufflePatterns
  (params (ShouldShuffle :type Boolean))
  (if ShouldShuffle
    (Networking|SwitchHasAuthority
      (:Authority
        (if (not (Variables|Default|GetDebugPatternEnabled))
          (Utilities|Array|Shuffle (Variables|Default|GetFountainPatterns)))))))

(fn ConstructionScript
  (if (Variables|Default|GetShowDebugArrowsInGame)
    (for _array_element (Components|GetChildrenComponents (Variables|Default|GetDefaultSceneRoot) true)
      (bind _asarrow_component (Utilities|Casting|CastToArrowComponent _array_element))
      (Development|SetHiddeninGame _asarrow_component false false))
    (else
      (for _array_element_2 (Components|GetChildrenComponents (Variables|Default|GetDefaultSceneRoot) true)
        (bind _asarrow_component_2 (Utilities|Casting|CastToArrowComponent _array_element_2))
        (Development|SetHiddeninGame _asarrow_component_2 true false))))
  (if (Variables|Default|GetShowDebugArrowsInEditor)
    (for _array_element_3 (Components|GetChildrenComponents (Variables|Default|GetDefaultSceneRoot) true)
      (bind _asarrow_component_3 (Utilities|Casting|CastToArrowComponent _array_element_3))
      (Rendering|SetVisibility _asarrow_component_3 true false))
    (else
      (for _array_element_4 (Components|GetChildrenComponents (Variables|Default|GetDefaultSceneRoot) true)
        (bind _asarrow_component_4 (Utilities|Casting|CastToArrowComponent _array_element_4))
        (Rendering|SetVisibility _asarrow_component_4 false false))))
  (for _array_element_5 (Components|GetChildrenComponents (Variables|Default|GetDefaultSceneRoot) true)
    (bind _asarrow_component_5 (Utilities|Casting|CastToArrowComponent _array_element_5))
    (Utilities|Array|AddUnique (LocalVariables|GetLocalNozzleTransforms) (Math|Transform|MakeTransform (- (.location (Transformation|GetWorldTransform _asarrow_component_5)) "0.000000,0.000000,60.000000") "0, 0, 0" "0.350000,0.350000,0.800000")))
  (Components|InstancedStaticMesh|AddInstances (Variables|Default|GetISM_FountainNozzles) (LocalVariables|GetLocalNozzleTransforms) false true true)
  (Components|InstancedStaticMesh|AddInstances (Variables|Default|GetISM_FountainNozzleBases) (LocalVariables|GetLocalNozzleTransforms) false true true))
